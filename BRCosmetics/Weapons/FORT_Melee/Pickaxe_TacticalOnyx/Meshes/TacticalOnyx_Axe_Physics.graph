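# Physics asset for the TacticalOnyx axe pickaxe cosmetic, defining a jointed chain of dyn_tactical bones on the melee weapon mesh.
# A long capsule on dyn_tactical_base anchors five box bodies, dyn_tactical_01 through 05, each constrained to the one before it.
# Swing limits widen down the chain, with the second axis leveling off from the third joint, while every twist limit is zero.
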PhysicsAsset /BRCosmetics/Weapons/FORT_Melee/Pickaxe_TacticalOnyx/Meshes/TacticalOnyx_Axe_Physics
bodies {
 dyn_tactical_01 [{
   shapeType=Box
   extentX=1.688
   extentY=7.1046
   extentZ=18.4754
   shapeName=dyn_tactical_01_box
   center=<2.8001 0.042 -1.5426>
   rotation=<rot 0 -90 0.118>
  }]
 dyn_tactical_02 [{
   shapeType=Box
   extentX=1.688
   extentY=7.4827
   extentZ=18.5165
   shapeName=dyn_tactical_02_box
   center=<2.8883 0.042 -1.5421>
   rotation=<rot 0 -90 0>
  }]
 dyn_tactical_03 [{
   shapeType=Box
   extentX=1.688
   extentY=9.5353
   extentZ=18.1993
   shapeName=dyn_tactical_03_box
   center=<3.8996 0.042 -1.537>
   rotation=<rot 0 -90 -0.0761>
  }]
 dyn_tactical_04 [{
   shapeType=Box
   extentX=1.688
   extentY=8.8503
   extentZ=18.1991
   shapeName=dyn_tactical_04_box
   center=<6.4159 0.042 -1.5311>
   rotation=<rot 0 -90 0>
  }]
 dyn_tactical_05 [{
   shapeType=Box
   extentX=1.688
   extentY=7.4966
   extentZ=17.9364
   shapeName=dyn_tactical_05_box
   center=<6.6873 0.042 -1.5308>
   rotation=<rot 0 -90 -0.0048>
  }]
 dyn_tactical_base [{
   shapeType=Capsule
   length=87.7317
   radius=3.6682
   shapeName=dyn_tactical_base_capsule
   center=<0.6373 0.0112 -26.1683>
   rotation=<rot -0.0158 -90 0.8691>
  }]
}
constraints [
 {bone1Name=dyn_tactical_01, bone2Name=dyn_tactical_base, swing1LimitDegrees=20, swing2LimitDegrees=8, twistLimitDegrees=0}
 {bone1Name=dyn_tactical_02, bone2Name=dyn_tactical_01, swing1LimitDegrees=30, swing2LimitDegrees=12, twistLimitDegrees=0}
 {bone1Name=dyn_tactical_03, bone2Name=dyn_tactical_02, swing1LimitDegrees=40, swing2LimitDegrees=15, twistLimitDegrees=0}
 {bone1Name=dyn_tactical_04, bone2Name=dyn_tactical_03, swing1LimitDegrees=50, swing2LimitDegrees=15, twistLimitDegrees=0}
 {bone1Name=dyn_tactical_05, bone2Name=dyn_tactical_04, swing1LimitDegrees=60, swing2LimitDegrees=15, twistLimitDegrees=0}
]
