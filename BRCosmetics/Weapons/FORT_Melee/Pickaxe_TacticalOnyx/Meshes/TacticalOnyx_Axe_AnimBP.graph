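# Animation Blueprint for the TacticalOnyx pickaxe, adding secondary motion to its dyn_tactical bone chain.
# The cached input pose feeds either RigidBody simulation in base-bone space or additive bone rotations, some front-end only, followed by a Trail Controller.
# RigidsActive selects RigidBody when rigid bodies are enabled or LOD exceeds 1, otherwise the trail branch.
# The update event also sets two front-end alpha variables that the AnimGraph never reads.
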AnimBlueprint /BRCosmetics/Weapons/FORT_Melee/Pickaxe_TacticalOnyx/Meshes/TacticalOnyx_Axe_AnimBP
target_skeleton=&/TacticalOnyx_Axe_Skeleton
graphs [
 AnimGraph {nodes [
   n1 = RigidBody(ComponentPose=Local To Component(LocalPose=Use cached pose 'pre_dynamics'())) {
    bOverrideWorldGravity=true
    overridePhysicsAsset=&/TacticalOnyx_Axe_Physics
    simulationSpace=BaseBoneSpace
    alphaBoolBlend {blendOption=Linear}
    baseBoneRef {boneName=dyn_tactical_base}
    componentAppliedLinearAccClamp=<5000 5000 5000>
    componentLinearAccScale=<1 1 1>
    componentLinearVelScale=<1 1 1>
    overrideWorldGravity {z=-980}
   }
   n2 = Get bIsInFrontEnd()
   n3 = "Transform (Modify) Bone - Bone: dyn_tactical_01"!(
    ComponentPose=Local To Component(LocalPose=Use cached pose 'pre_dynamics'())
    bAlphaBoolEnabled:bool=@n2
    Translation="0,0,0"
    Rotation="-10,100,0"
    Scale="1,1,1"
   ) -> (Pose) {
    alphaInputType=Bool
    rotationMode=BMM_Additive
    rotationSpace=BCS_BoneSpace
    alphaBoolBlend {blendOption=Linear}
    boneToModify {boneName=dyn_tactical_01}
    rotation {pitch=-10, yaw=100}
   }
   n4 = "Transform (Modify) Bone - Bone: dyn_tactical_02"!(ComponentPose=@n3, bAlphaBoolEnabled:bool=@n2, Translation="0,0,0", Rotation="-10,30,0", Scale="1,1,1") -> (Pose) {
    alphaInputType=Bool
    rotationMode=BMM_Additive
    rotationSpace=BCS_BoneSpace
    alphaBoolBlend {blendOption=Linear}
    boneToModify {boneName=dyn_tactical_02}
    rotation {pitch=-10, yaw=30}
   }
   n5 = "Transform (Modify) Bone - Bone: dyn_tactical_03"!(ComponentPose=@n4, bAlphaBoolEnabled:bool=@n2, Translation="0,0,0", Rotation="-20,40,0", Scale="1,1,1") -> (Pose) {
    alphaInputType=Bool
    rotationMode=BMM_Additive
    rotationSpace=BCS_BoneSpace
    alphaBoolBlend {blendOption=Linear}
    boneToModify {boneName=dyn_tactical_03}
    rotation {pitch=-20, yaw=40}
   }
   n6 = "Transform (Modify) Bone - Bone: dyn_tactical_04"!(ComponentPose=@n5, bAlphaBoolEnabled:bool=@n2, Translation="0,0,0", Rotation="-20,0,0", Scale="1,1,1") -> (Pose) {
    alphaInputType=Bool
    rotationMode=BMM_Additive
    rotationSpace=BCS_BoneSpace
    alphaBoolBlend {blendOption=Linear}
    boneToModify {boneName=dyn_tactical_04}
    rotation {pitch=-20}
   }
   n7 = "Transform (Modify) Bone - Bone: dyn_tactical_01"(ComponentPose=@n6, bAlphaBoolEnabled=True, Translation="0,0,0", Rotation="-10,-60,0", Scale="1,1,1") {
    alphaInputType=Bool
    rotationMode=BMM_Additive
    rotationSpace=BCS_BoneSpace
    scaleMode=BMM_Additive
    scaleSpace=BCS_BoneSpace
    translationMode=BMM_Additive
    translationSpace=BCS_BoneSpace
    alphaBoolBlend {blendOption=Linear}
    boneToModify {boneName=dyn_tactical_01}
    rotation {pitch=-10, yaw=-60}
   }
   n8 = Trail controller - Bone: dyn_tactical_05!(ComponentPose=@n7, Alpha=0.6, RelaxationSpeedScale:float=1, FakeVelocity:Vector="0,0,0") -> (Pose) {
    bShowTrailLocation=true
    bUsePlanarLimit=true
    chainLength=5
    debugLifeTime=0.1
    maxDeltaTime=0.0333
    relaxationSpeedScale=1.2
    alphaScaleBiasClamp {interpSpeedDecreasing=25, interpSpeedIncreasing=25}
    rotationLimits [{limitMax=<180 180 180>}, {limitMax=<180 180 180>}, {limitMax=<180 180 180>}, {limitMax=<180 180 180>}, {limitMax=<180 180 180>}]
    rotationOffsets [<0 0 0>, <0 0 0>, <0 0 0>, <0 0 0>, <0 0 0>]
    trailBone {boneName=dyn_tactical_05}
    trailRelaxationSpeed {editorCurveData {keys [<at 0 15 RCIM_Linear>,<at 1 6 RCIM_Linear>]}}
   }
   Output Pose(Result=Blend Poses by bool(
     BlendPose_0=Component To Local(ComponentPose=@n1)
     BlendPose_1=Component To Local(ComponentPose=@n8)
     bActiveValue=Get RigidsActive() -> (RigidsActive:bool)
    ) {blendType=Linear}) {layerGroup=None}
   AnimGraphNode_SubInput_0 = Input Pose()
   Save cached pose 'pre_dynamics'(Pose=@AnimGraphNode_SubInput_0) {cacheName=pre_dynamics}
   Trail controller - Bone: dyn_tactical_05!(
    ComponentPose=Reroute Node(InputPin:Component Space Pose Link Structure) -> (OutputPin:Component Space Pose Link Structure)
    bAlphaBoolEnabled:bool=@n2
   ) -> (Pose) {
    alphaInputType=Bool
    chainLength=4
    relaxationSpeedScale=2.5
    rotationLimits [{limitMax=<180 180 180>}, {limitMax=<180 180 180>}, {limitMax=<180 180 180>}, {limitMax=<180 180 180>}]
    rotationOffsets [<0 0 0>, <0 0 0>, <0 0 0>, <0 0 0>]
    trailBone {boneName=dyn_tactical_05}
    bLimitStretch=~
   }
  ]}
 EventGraph {nodes [
   Event BlueprintUpdateAnimation()
   > n1 = Sequence()
   n2 = Get CurrentLOD()
   > n3 = Branch(Condition=OR Boolean(A=Get bRigidBodyEnabled(), B="integer > integer"(A=@n2, B=1) {is=title K2Node_CallFunction}))
   > Set RigidsActive(RigidsActive:bool=true) -> (Output_Get:bool)
   Set RigidsActive(execute=@n3.else, RigidsActive=false)
   n4 = Branch(execute=@n1.then_1, Condition=Get bIsInFrontEnd())
   > Set RigidsFE_Alpha(RigidsFE_Alpha:double=1) -> (Output_Get:double)
   > Set NegPhysicsAlpha(NegPhysicsAlpha:double=0) -> (Output_Get:double)
   Set RigidsFE_Alpha(execute=@n4.else, RigidsFE_Alpha=0)
  ]}
]
variables [
 NegPhysicsAlpha {type=number}
 RigidsActive {type=boolean}
 RigidsFE_Alpha {type=number}
 rotateRun {type=boolean}
]
